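; Character part modifier for the Donut Cup large male body cosmetic that toggles its idle VFX when the player aims down sights.
; On the ADS bool event it gets the idle VFX component, calls the parent logic, then deactivates the component while targeting and reactivates it with reset otherwise.
; The construction script only calls the parent.
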(event EventADSBoolLogic
  (params (isTargeting :type Boolean))
  (bind _returnvalue (CallFunction|GetIdleVFXComponent))
  (|CallParentFunction isTargeting)
  (if isTargeting
    (Components|Activation|Deactivate _returnvalue)
    (else
      (Components|Activation|Activate _returnvalue true))))

(fn ConstructionScript
  (|CallParentFunction))
