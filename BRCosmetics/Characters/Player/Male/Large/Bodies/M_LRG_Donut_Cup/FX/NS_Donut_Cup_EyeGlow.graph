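# Niagara eye-glow effect for the Donut Cup large male player cosmetic body, pinned to the character's skeletal mesh with local offsets.
# Glowing_Eye_Flare bursts three spawn groups into sprites and a light with varied size, offset and visibility; Sketch_Flare adds a size-jittered SubUV sprite.
# Both pulse alpha over particle life, gated by System LOD Scaling thresholds, which also toggle the light.
# User.bisHeadHidden overrides each emitter's visibility tag.
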NiagaraSystem /BRCosmetics/Characters/Player/Male/Large/Bodies/M_LRG_Donut_Cup/FX/NS_Donut_Cup_EyeGlow
dependencies {
 usedDataInterfaces [
  &NiagaraDataInterfaceCurve
  &NiagaraDataInterfaceSpriteRendererInfo
  &NiagaraDataInterfaceSkeletalMesh
 ]
 usedDynamicInputs [&FN_LOD_Distance_Remap]
 usedModules [
  &SystemState
  &self:SystemSpawnScript.SystemScriptSource.SystemScriptGraph.NiagaraNodeAssignment_0.SetVariables_#1
  &self:SystemSpawnScript.SystemScriptSource.SystemScriptGraph.NiagaraNodeAssignment_1.SetVariables_#2
 ]
}
emitters {
 Glowing_Eye_Flare {
  emitterUpdateScript [
   &EmitterState()
   &SpawnBurst_Instantaneous(Use Spawn Probability=false, UseLoopCountLimit=false)
   &SpawnBurst_Instantaneous(Spawn Group=1, Use Spawn Probability=false, UseLoopCountLimit=false) {module=SpawnBurst_Instantaneous001}
   &SpawnBurst_Instantaneous(Spawn Group=2, Use Spawn Probability=false, UseLoopCountLimit=false) {module=SpawnBurst_Instantaneous002}
  ]
  particleSpawnScript [
   /SetVariables(Particles.VisibilityTag=&MaskIntBySpawnGroup(Masked Int=1, Passthrough Int=0, Spawn Group Mask=2)) {isSetParameters=true}
   &InitializeParticle(
    Color=<rgba 4 2 1 0.1>
    Sprite Size=&MaskVector2dBySpawnGroup(Masked Vector2D=<20 28>, Passthrough Vector2D=<12 8>, Spawn Group Mask=0)
    Sprite Size Mode=::NewEnumerator4
    UsePositionOffset=false
    Write Material Random=true
   )
   &CameraOffset(Camera Offset Amount=4)
  ]
  particleUpdateScript [
   &ParticleState()
   &ScaleColor(
    Scale Alpha=&FloatFromCurve(
     FloatCurve=DataInterface() {propertyValues {Curve {keys [<at 0 0>,<at 0.3 1>,<at 1 0>]}}}
     Scale Curve=&MakeCustomFloatFromBool(
      Bool=&SetBoolByFloatComparison(A=$System.LOD Scaling, B=0.9999, Comparison Type=::NewEnumerator2)
      True Float=0
      False Float=1
     )
    )
    ScaleA=true
    ScaleRGB=false
   )
   &SkeletalMeshLocation(
    Bone Sampling Mode=::NewEnumerator4
    MaskSpawnGroup=false
    Sampled Position Offset=&MaskVectorBySpawnGroup(Masked Vector=<4.1 7.25 -4.5>, Passthrough Vector=<4.6 6.1 -3>, Spawn Group Mask=2)
    Skeletal Mesh=$System.Skeletal Mesh
   )
   &Light_Attributes(
    Color=1
    Color Mode=0
    Enabled=1
    Falloff Exponent=1
    Light Color:LinearColor=&MultiplyLinearColor(
     Linear Color=<rgba 4 2.5161 0.4398 4>
     Scale Factor=&MakeVector4FromVectorAndFloat(
      "Vector (RGB)"=<1 1 1>
      "Float (Alpha)"=&FloatFromCurve(FloatCurve=DataInterface() {propertyValues {Curve {keys [<at 0 0>,<at 0.3 1>,<at 1 0>]}}})
     )
    )
    Light Enabled:NiagaraBool=&SetBoolByFloatComparison(A=$System.LOD Scaling, B=0.9999, Comparison Type=::NewEnumerator6)
    Light Exponent=1.5
    Light Position Offset:Vector3f=<-2 4 -1.5>
    Light Radius=4.8
    Position=1
    Position Offset Coordinate Space:ENiagaraCoordinateSpace=::Local
    Radius=1
    UsePositionOffset=true
   )
   /SetVariables(Particles.VisibilityTag=&MakeCustomIntFromBool(Bool=$User.bisHeadHidden, True INT=2, False INT=$Particles.VisibilityTag)) {isSetParameters=true, module=SetVariables_#3}
  ]
  rendererClasses [&NiagaraSpriteRendererProperties, &NiagaraLightRendererProperties]
  renderers [
   {rendererClass=&NiagaraSpriteRendererProperties}
   {rendererClass=&NiagaraLightRendererProperties, rendererIndex=1}
  ]
 }
 Sketch_Flare {
  emitterUpdateScript [
   &EmitterState()
   &SpawnBurst_Instantaneous(Use Spawn Probability=false, UseLoopCountLimit=false)
  ]
  particleSpawnScript [
   &InitializeParticle(
    Color=<rgba 4 3 0.5 0.7>
    Lifetime=0.99
    Sprite Rotation Angle=-0.1
    Sprite Rotation Mode=::NewEnumerator1
    Sprite Size=<20 10>
    Sprite Size Mode=::NewEnumerator4
    UsePositionOffset=false
    Write Material Random=true
   )
   &SubUVAnimation(
    "Sprite Renderer "=DataInterface() {propertyValues {SpriteRenderer=&self:Sketch_Flare_0.NiagaraSpriteRendererProperties_0}}
    SubUV Animation Mode:ENiagara_SubUVLookupModeV2=::NewEnumerator5
   )
   &CameraOffset(Camera Offset Amount=4)
  ]
  particleUpdateScript [
   &ParticleState()
   &ScaleSpriteSize(
    Scale Sprite Size Mode=::NewEnumerator0
    Uniform Scale Factor=&RandomRangeFloat(Minimum=0.6, Maximum=0.8, Evaluation Type=::NewEnumerator1)
   )
   &SkeletalMeshLocation(Bone Sampling Mode=::NewEnumerator4, MaskSpawnGroup=false, Sampled Position Offset=<4.4 7 -3>, Skeletal Mesh=$System.Skeletal Mesh)
   &ScaleColor(
    Scale Alpha=&FloatFromCurve(
     FloatCurve=DataInterface() {propertyValues {Curve {keys [<at 0 0>,<at 0.3 1>,<at 1 0>]}}}
     Scale Curve=&MakeCustomFloatFromBool(
      Bool=&SetBoolByFloatComparison(A=$System.LOD Scaling, B=0.6667, Comparison Type=::NewEnumerator2)
      True Float=1
      False Float=0
     )
    )
    ScaleA=true
    ScaleRGB=false
   )
   /SetVariables(Particles.VisibilityTag=&MakeCustomIntFromBool(Bool=$User.bisHeadHidden, True INT=1, False INT=0)) {isSetParameters=true, module=SetVariables_#3}
  ]
  rendererClasses [&NiagaraSpriteRendererProperties]
  renderers [{rendererClass=&NiagaraSpriteRendererProperties}]
 }
}
summary {
 systemName=NS_Donut_Cup_EyeGlow
 emitters [
  {
   emitterName=Glowing_Eye_Flare
   rendererClasses [&NiagaraSpriteRendererProperties, &NiagaraLightRendererProperties]
  }
  {emitterName=Sketch_Flare, rendererClasses [&NiagaraSpriteRendererProperties]}
 ]
 userVariables [User.bisHeadHidden {type=NiagaraBool}]
}
