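# Post-process AnimBP for the M_LRG_Donut_Cup large male cosmetic body, adding corrective and secondary motion to gear and muscles.
# AnimGraph copies the parent mesh pose, then chains weighted constraints, a spring strap, pectoral Look Ats, shoepad flaps and additive bone tweaks.
# EventGraph derives blend alphas from limb twist relative to reference pose, plus targeting, crouch, rifle-aim and backbling states.
# Several AnimGraph chains, including arm-muscle dynamics, never reach Output Pose.
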AnimBlueprint /BRCosmetics/Characters/Player/Male/Large/Bodies/M_LRG_Donut_Cup/Meshes/M_LRG_Donut_Cup_AnimBP
target_skeleton=&/M_LRG_Donut_Cup_Skeleton
graphs [
 AnimGraph {nodes [
   n1 = Get RflexAlpha() -> (RflexAlpha:double)
   n2 = Anim Dynamics - Bone: tricep_01_r!(ComponentPose, Alpha=@n1, GravityOverride:Vector="0,0,0", ExternalForce="0,0,0") -> (Pose) {
    angularDampingOverride=1
    angularSpringConstant=1500
    bAngularSpring=true
    bGravityOverrideInSimSpace=true
    bLinearSpring=true
    bOverrideAngularDamping=true
    gravityScale=0
    linearDampingOverride=1
    linearSpringConstant=500
    simulationSpace=RootRelative
    alphaBoolBlend {blendOption=Linear}
    boundBone {boneName=tricep_01_r}
    physicsBodyDefinitions [{
      sphereCollisionRadius=0
      boundBone {boneName=tricep_01_r}
      boxExtents=<10 20 20>
      constraintSetup {angularLimitsMax=<20 5 5>, angularLimitsMin=<-20 -5 -5>, angularTarget=<0 0 0>, linearAxesMax=<1 1 1>, linearAxesMin=<-1 -1 -1>}
      localJointOffset=<10 0 0>
     }]
   }
   n3 = Anim Dynamics - Bone: tricep_02_r(ComponentPose=@n2, Alpha=@n1, ExternalForce="0,0,0") {
    angularDampingOverride=1
    angularSpringConstant=1500
    bAngularSpring=true
    bGravityOverrideInSimSpace=true
    bLinearSpring=true
    bOverrideAngularDamping=true
    gravityScale=0
    linearDampingOverride=1
    linearSpringConstant=500
    simulationSpace=RootRelative
    alphaBoolBlend {blendOption=Linear}
    boundBone {boneName=tricep_02_r}
    physicsBodyDefinitions [{
      sphereCollisionRadius=0
      boundBone {boneName=tricep_02_r}
      boxExtents=<10 20 20>
      constraintSetup {angularLimitsMax=<20 5 5>, angularLimitsMin=<-20 -5 -5>, angularTarget=<0 0 0>, linearAxesMax=<1 1 1>, linearAxesMin=<-1 -1 -1>}
      localJointOffset=<10 0 0>
     }]
   }
   Anim Dynamics - Bone: bicep_01_r(ComponentPose=@n3, Alpha=@n1, ExternalForce="0,0,0") {
    angularDampingOverride=1
    angularSpringConstant=1500
    bAngularSpring=true
    bGravityOverrideInSimSpace=true
    bLinearSpring=true
    bOverrideAngularDamping=true
    gravityScale=0
    linearDampingOverride=1
    linearSpringConstant=500
    simulationSpace=RootRelative
    alphaBoolBlend {blendOption=Linear}
    boundBone {boneName=bicep_01_r}
    physicsBodyDefinitions [{
      sphereCollisionRadius=0
      boundBone {boneName=bicep_01_r}
      boxExtents=<10 20 20>
      constraintSetup {angularLimitsMax=<20 5 5>, angularLimitsMin=<-20 -5 -5>, angularTarget=<0 0 0>, linearAxesMax=<1 1 1>, linearAxesMin=<-1 -1 -1>}
      localJointOffset=<10 0 0>
     }]
   }
   "Constraint - dyn_parent_shldrpad_l (PP)"!(ComponentPose, Alpha=1, ConstraintWeights_0=0.5, ConstraintWeights_1:float=0.5) -> (Pose) {
    alphaBoolBlend {blendOption=Linear}
    boneToModify {boneName=dyn_parent_shldrpad_l}
    constraintSetup [{targetBone {boneName=spine_05}}, {targetBone {boneName=clavicle_l}}]
    constraintWeights [1,1]
   }
   n4 = Get LshoulderRotateBackAlpha() -> (LshoulderRotateBackAlpha:double)
   n5 = "Transform (Modify) Bone - Bone: dyn_driver_shldrpad_l"(Alpha=@n4, Translation="0,0,0", Rotation="0.000,0.000,35", Scale="1,1,1") {
    rotationMode=BMM_Additive
    rotationSpace=BCS_BoneSpace
    alphaBoolBlend {blendOption=Linear}
    boneToModify {boneName=dyn_driver_shldrpad_l}
    rotation {roll=35}
   }
   n6 = "Transform (Modify) Bone - Bone: dyn_shldrpad_01_l"(ComponentPose=@n5, Alpha=@n4, Translation="4.831,3.092,3.18", Rotation="0.000,-25.001,0.000", Scale="1,1,1") {
    rotationMode=BMM_Additive
    rotationSpace=BCS_BoneSpace
    translationMode=BMM_Additive
    translationSpace=BCS_BoneSpace
    alphaBoolBlend {blendOption=Linear}
    boneToModify {boneName=dyn_shldrpad_01_l}
    rotation {yaw=-25.0007}
    translation=<4.831 3.092 3.18>
   }
   n7 = "Transform (Modify) Bone - Bone: dyn_driver_shldrpad_l"(ComponentPose=@n6, Alpha=@n4, Translation="0,0,0", Rotation="0.000,0.000,20.000", Scale="1,1,1") {
    rotationMode=BMM_Additive
    rotationSpace=BCS_BoneSpace
    translationMode=BMM_Additive
    translationSpace=BCS_BoneSpace
    alphaBoolBlend {blendOption=Linear}
    boneToModify {boneName=dyn_driver_shldrpad_l}
    rotation {roll=20}
   }
   n8 = "Transform (Modify) Bone - Bone: bicep_00_l"(ComponentPose=@n7, Alpha=@n4, Translation="0,0,0", Rotation="0,0,0", Scale="1,1,1") {
    rotationMode=BMM_Additive
    rotationSpace=BCS_BoneSpace
    translationMode=BMM_Additive
    translationSpace=BCS_BoneSpace
    alphaBoolBlend {blendOption=Linear}
    boneToModify {boneName=bicep_00_l}
   }
   n9 = "Transform (Modify) Bone - Bone: bicep_00_lower_l"(ComponentPose=@n8, Alpha=@n4, Translation="0,0,0", Rotation="0.000,0.000,55", Scale="1,1,1") {
    rotationMode=BMM_Additive
    rotationSpace=BCS_BoneSpace
    translationMode=BMM_Additive
    translationSpace=BCS_BoneSpace
    alphaBoolBlend {blendOption=Linear}
    boneToModify {boneName=bicep_00_lower_l}
    rotation {roll=55}
   }
   Look At - Bone: dyn_shldrpad_01_l(ComponentPose=@n9, Alpha=@n4) {
    alphaBoolBlend {blendOption=Linear}
    boneToModify {boneName=dyn_shldrpad_01_l}
    lookAtLocation {x=0}
    lookAtTarget {boneReference {boneName=dyn_shldrpad_01_l_aim}}
    lookAt_Axis {axis=<1 0>}
    lookUp_Axis {axis {x=0, z=1}}
   }
   "Transform (Modify) Bone - Bone: dyn_shldrpad_01_l"(Alpha=@n4, Translation="-0.000,-0.000,2", Rotation="0,0,0", Scale="1,1,1") {
    rotationMode=BMM_Additive
    rotationSpace=BCS_BoneSpace
    translationMode=BMM_Additive
    translationSpace=BCS_BoneSpace
    alphaBoolBlend {blendOption=Linear}
    boneToModify {boneName=dyn_shldrpad_01_l}
    translation {z=2}
   }
   "Transform (Modify) Bone - Bone: dyn_shldrpad_01_l"(
    Alpha=Get LshoulderRotateBackAlpha_Backbling() -> (LshoulderRotateBackAlpha_Backbling:double)
    Translation="5.703,0.000,0.000"
    Rotation="0,0,0"
    Scale="1,1,1"
   ) {
    rotationMode=BMM_Additive
    rotationSpace=BCS_BoneSpace
    translationMode=BMM_Additive
    translationSpace=BCS_BoneSpace
    alphaBoolBlend {blendOption=Linear}
    alphaScaleBiasClamp {bInterpResult=true}
    boneToModify {boneName=dyn_shldrpad_01_l}
    translation=<5.7035 0 0>
   }
   n10 = Get RarmDistanceAlpha() -> (RarmDistanceAlpha:double)
   n11 = "Transform (Modify) Bone - Bone: tricep_02_r"(Alpha=@n10, Translation="2.160,0.045,-0.934", Rotation="-10.000,0.000,0.000", Scale="1.15,1.15,1.15") {
    rotationMode=BMM_Additive
    rotationSpace=BCS_BoneSpace
    scaleMode=BMM_Additive
    scaleSpace=BCS_BoneSpace
    translationMode=BMM_Additive
    translationSpace=BCS_BoneSpace
    alphaBoolBlend {blendOption=Linear}
    boneToModify {boneName=tricep_02_r}
    rotation {pitch=-10}
    scale=<1.15 1.15 1.15>
    translation=<2.1601 0.045 -0.9341>
   }
   "Transform (Modify) Bone - Bone: tricep_01_r"(ComponentPose=@n11, Alpha=@n10, Translation="2.644,-0.045,1.855", Rotation="0,0,0", Scale="1.15,1.15,1.15") {
    rotationMode=BMM_Additive
    rotationSpace=BCS_BoneSpace
    scaleMode=BMM_Additive
    scaleSpace=BCS_BoneSpace
    translationMode=BMM_Additive
    translationSpace=BCS_BoneSpace
    alphaBoolBlend {blendOption=Linear}
    boneToModify {boneName=tricep_01_r}
    scale=<1.15 1.15 1.15>
    translation=<2.644 -0.045 1.855>
   }
   n12 = Get RshoulderRotateBackAlpha() -> (RshoulderRotateBackAlpha:double)
   n13 = "Transform (Modify) Bone - Bone: dyn_driver_shldrpad_r"(Alpha=@n12, Translation="0,0,0", Rotation="0.000,0.000,-10.000", Scale="1,1,1") {
    rotationMode=BMM_Additive
    rotationSpace=BCS_BoneSpace
    alphaBoolBlend {blendOption=Linear}
    boneToModify {boneName=dyn_driver_shldrpad_r}
    rotation {roll=-10.0002}
   }
   n14 = "Transform (Modify) Bone - Bone: dyn_shldrpad_01_r"(ComponentPose=@n13, Alpha=@n12, Translation="1.403,-5.237,2.965", Rotation="0.000,15.000,20.000", Scale="1,1,1") {
    rotationMode=BMM_Additive
    rotationSpace=BCS_BoneSpace
    translationMode=BMM_Additive
    translationSpace=BCS_BoneSpace
    alphaBoolBlend {blendOption=Linear}
    boneToModify {boneName=dyn_shldrpad_01_r}
    rotation {roll=20, yaw=15.0001}
    translation=<1.403 -5.237 2.9649>
   }
   n15 = Look At - Bone: dyn_shldrpad_01_r(ComponentPose=@n14, Alpha=@n12) {
    alphaBoolBlend {blendOption=Linear}
    boneToModify {boneName=dyn_shldrpad_01_r}
    lookAtLocation {x=0}
    lookAtTarget {boneReference {boneName=dyn_shldrpad_01_r_aim}}
    lookAt_Axis {axis=<1 0>}
    lookUp_Axis {axis {x=0, z=1}}
   }
   n16 = Get bIsTargeting()
   n17 = "Transform (Modify) Bone - Bone: dyn_shldrpad_01_r"!(ComponentPose=@n15, bAlphaBoolEnabled:bool=@n16, Translation="0,0,0", Rotation="10,0.000,0.000", Scale="1,1,1") -> (Pose) {
    alphaInputType=Bool
    rotationMode=BMM_Additive
    rotationSpace=BCS_BoneSpace
    translationMode=BMM_Additive
    translationSpace=BCS_BoneSpace
    alphaBoolBlend {blendOption=Linear}
    boneToModify {boneName=dyn_shldrpad_01_r}
    rotation {pitch=10}
   }
   n18 = "Transform (Modify) Bone - Bone: dyn_shldrpad_03_r"!(ComponentPose=@n17, bAlphaBoolEnabled:bool=@n16, Translation="-0.000,-5.216,0.000", Rotation="9.392,-3.451,-20.283", Scale="1,1,1") -> (Pose) {
    alphaInputType=Bool
    rotationMode=BMM_Additive
    rotationSpace=BCS_BoneSpace
    scaleMode=BMM_Replace
    scaleSpace=BCS_BoneSpace
    translationMode=BMM_Additive
    translationSpace=BCS_BoneSpace
    alphaBoolBlend {blendOption=Linear}
    boneToModify {boneName=dyn_shldrpad_03_r}
    rotation=<rot 9.392 -3.451 -20.2829>
    translation {y=-5.216}
   }
   "Transform (Modify) Bone - Bone: dyn_shldrpad_03_r"(ComponentPose=@n18, bAlphaBoolEnabled=@n16, Translation="0,0,0", Rotation="0,0,0", Scale="0.9,1,1") {
    alphaInputType=Bool
    rotationMode=BMM_Additive
    rotationSpace=BCS_BoneSpace
    scaleMode=BMM_Replace
    scaleSpace=BCS_BoneSpace
    translationMode=BMM_Additive
    translationSpace=BCS_BoneSpace
    alphaBoolBlend {blendOption=Linear}
    boneToModify {boneName=dyn_shldrpad_03_r}
    scale {x=0.9}
   }
   n19 = Get RbicepFlex() -> (RbicepFlex:double)
   "Transform (Modify) Bone - Bone: bicep_02_r"(Alpha=@n19, Translation="-4.899,-0.000,-0.000", Rotation="0,0,0", Scale="1.000,1.000,2.000") {
    rotationMode=BMM_Additive
    rotationSpace=BCS_BoneSpace
    scaleMode=BMM_Additive
    scaleSpace=BCS_BoneSpace
    translationMode=BMM_Additive
    translationSpace=BCS_BoneSpace
    alphaBoolBlend {blendOption=Linear}
    boneToModify {boneName=bicep_02_r}
    scale {z=2}
    translation {x=-4.899}
   }
   "Transform (Modify) Bone - Bone: bicep_01_r"(Alpha=@n19, Translation="0,0,0", Rotation="0,0,0", Scale="1.5,1.5,1.5") {
    rotationMode=BMM_Additive
    rotationSpace=BCS_BoneSpace
    scaleMode=BMM_Additive
    scaleSpace=BCS_BoneSpace
    translationMode=BMM_Additive
    translationSpace=BCS_BoneSpace
    alphaBoolBlend {blendOption=Linear}
    alphaScaleBiasClamp {scale=0.4}
    boneToModify {boneName=bicep_01_r}
    scale=<1.5 1.5 1.5>
   }
   n20 = Get RarmDistanceAlpha()
   n21 = "Transform (Modify) Bone - Bone: bicep_02_r"(Alpha=@n20, Translation="-4.060,-0.000,2.927", Rotation="-25.000,0.000,0.000", Scale="1.750,1.750,1.750") {
    rotationMode=BMM_Additive
    rotationSpace=BCS_BoneSpace
    scaleMode=BMM_Additive
    scaleSpace=BCS_BoneSpace
    translationMode=BMM_Additive
    translationSpace=BCS_BoneSpace
    alphaBoolBlend {blendOption=Linear}
    boneToModify {boneName=bicep_02_r}
    rotation {pitch=-25}
    scale=<1.75 1.75 1.75>
    translation=<-4.0595 0 2.9268>
   }
   "Transform (Modify) Bone - Bone: bicep_01_r"(ComponentPose=@n21, Alpha=@n20, Translation="0,0,0", Rotation="0,0,0", Scale="1.5,1.5,1.5") {
    rotationMode=BMM_Additive
    rotationSpace=BCS_BoneSpace
    scaleMode=BMM_Additive
    scaleSpace=BCS_BoneSpace
    translationMode=BMM_Additive
    translationSpace=BCS_BoneSpace
    alphaBoolBlend {blendOption=Linear}
    boneToModify {boneName=bicep_01_r}
    scale=<1.5 1.5 1.5>
   }
   n22 = "Constraint - joint2 (PPPPPPPPPPP)"!(
    ComponentPose=Local To Component(LocalPose=Copy Pose From Mesh() {bUseAttachedParent=true})
    Alpha=1
    ConstraintWeights_0=0.0072
    ConstraintWeights_1:float=0.0003
    ConstraintWeights_2:float=0.012
    ConstraintWeights_3:float=0.0876
    ConstraintWeights_4:float=0.2104
    ConstraintWeights_5:float=0.3639
    ConstraintWeights_6:float=0.1533
    ConstraintWeights_7:float=0.0423
    ConstraintWeights_8:float=0.0007
    ConstraintWeights_9:float=0.1222
    ConstraintWeights_10:float=0.0002
   ) -> (Pose) {
    alphaBoolBlend {blendOption=Linear}
    boneToModify {boneName=joint2}
    constraintSetup [
     {targetBone {boneName=clavicle_l}}
     {targetBone {boneName=upperarm_twist_02_l}}
     {targetBone {boneName=pectoralis_02_l}}
     {targetBone {boneName=spine_01}}
     {targetBone {boneName=spine_02}}
     {targetBone {boneName=spine_03}}
     {targetBone {boneName=spine_04}}
     {targetBone {boneName=spine_05}}
     {targetBone {boneName=upperarm_twist_01_l}}
     {targetBone {boneName=pelvis}}
     {targetBone {boneName=upperarm_l}}
    ]
    constraintWeights [1,1,1,1,1,1,1,1,1,1,1]
   }
   n23 = "Constraint - joint3 (PPPPPPPPPPPP)"!(
    ComponentPose=@n22
    Alpha=1
    ConstraintWeights_0=0
    ConstraintWeights_1:float=0.0324
    ConstraintWeights_2:float=0.0005
    ConstraintWeights_3:float=0.213
    ConstraintWeights_4:float=0.0381
    ConstraintWeights_5:float=0.1638
    ConstraintWeights_6:float=0.2278
    ConstraintWeights_7:float=0.16
    ConstraintWeights_8:float=0.1038
    ConstraintWeights_9:float=0.0014
    ConstraintWeights_10:float=0.0588
    ConstraintWeights_11:float=0.0004
   ) -> (Pose) {
    alphaBoolBlend {blendOption=Linear}
    boneToModify {boneName=joint3}
    constraintSetup [
     {targetBone {boneName=bicep_02_l}}
     {targetBone {boneName=clavicle_l}}
     {targetBone {boneName=upperarm_twist_02_l}}
     {targetBone {boneName=pectoralis_02_l}}
     {targetBone {boneName=spine_01}}
     {targetBone {boneName=spine_02}}
     {targetBone {boneName=spine_03}}
     {targetBone {boneName=spine_04}}
     {targetBone {boneName=spine_05}}
     {targetBone {boneName=upperarm_twist_01_l}}
     {targetBone {boneName=pelvis}}
     {targetBone {boneName=upperarm_l}}
    ]
    constraintWeights [1,1,1,1,1,1,1,1,1,1,1,1]
   }
   n24 = "Constraint - joint4 (PPPPPPPPPPP)"!(
    ComponentPose=@n23
    Alpha=1
    ConstraintWeights_0=0.0622
    ConstraintWeights_1:float=0.0001
    ConstraintWeights_2:float=0.5472
    ConstraintWeights_3:float=0.0176
    ConstraintWeights_4:float=0.1442
    ConstraintWeights_5:float=0.0276
    ConstraintWeights_6:float=0.0276
    ConstraintWeights_7:float=0.1396
    ConstraintWeights_8:float=0.0003
    ConstraintWeights_9:float=0.0337
    ConstraintWeights_10:float=0
   ) -> (Pose) {
    alphaBoolBlend {blendOption=Linear}
    boneToModify {boneName=joint4}
    constraintSetup [
     {targetBone {boneName=clavicle_l}}
     {targetBone {boneName=upperarm_twist_02_l}}
     {targetBone {boneName=pectoralis_02_l}}
     {targetBone {boneName=spine_01}}
     {targetBone {boneName=spine_02}}
     {targetBone {boneName=spine_03}}
     {targetBone {boneName=spine_04}}
     {targetBone {boneName=spine_05}}
     {targetBone {boneName=upperarm_twist_01_l}}
     {targetBone {boneName=pelvis}}
     {targetBone {boneName=upperarm_l}}
    ]
    constraintWeights [1,1,1,1,1,1,1,1,1,1,1]
   }
   n25 = "Constraint - dyn_pouch_01_bt_r (PPPPP)"!(
    ComponentPose=@n24
    Alpha=1
    ConstraintWeights_0=0.347
    ConstraintWeights_1:float=0.2502
    ConstraintWeights_2:float=0.2044
    ConstraintWeights_3:float=0.0228
    ConstraintWeights_4:float=0.1756
   ) -> (Pose) {
    alphaBoolBlend {blendOption=Linear}
    boneToModify {boneName=dyn_pouch_01_bt_r}
    constraintSetup [
     {targetBone {boneName=pelvis}}
     {targetBone {boneName=spine_02}}
     {targetBone {boneName=spine_03}}
     {targetBone {boneName=spine_04}}
     {targetBone {boneName=spine_01}}
    ]
    constraintWeights [1,1,1,1,1]
   }
   n26 = "Constraint - joint1 (PPPPP)"!(
    ComponentPose=@n25
    Alpha=1
    ConstraintWeights_0=0.2631
    ConstraintWeights_1:float=0.2513
    ConstraintWeights_2:float=0.2811
    ConstraintWeights_3:float=0.0264
    ConstraintWeights_4:float=0.1781
   ) -> (Pose) {
    alphaBoolBlend {blendOption=Linear}
    boneToModify {boneName=joint1}
    constraintSetup [
     {targetBone {boneName=pelvis}}
     {targetBone {boneName=spine_02}}
     {targetBone {boneName=spine_03}}
     {targetBone {boneName=spine_04}}
     {targetBone {boneName=spine_01}}
    ]
    constraintWeights [1,1,1,1,1]
   }
   n27 = "Constraint - pouch_fr_01_r (P)"!(ComponentPose=@n26, Alpha=1, ConstraintWeights_0=1, ConstraintWeights_1:float=0) -> (Pose) {
    alphaBoolBlend {blendOption=Linear}
    boneToModify {boneName=pouch_fr_01_r}
    constraintSetup [{targetBone {boneName=pelvis}}]
    constraintWeights [1,0]
   }
   n28 = "Constraint - pouch_01_r (PPPP)"!(
    ComponentPose=@n27
    Alpha=1
    ConstraintWeights_0=0.9999
    ConstraintWeights_1:float=0
    ConstraintWeights_2:float=0
    ConstraintWeights_3:float=0
   ) -> (Pose) {
    alphaBoolBlend {blendOption=Linear}
    boneToModify {boneName=pouch_01_r}
    constraintSetup [{targetBone {boneName=pelvis}}, {targetBone {boneName=spine_02}}, {targetBone {boneName=spine_03}}, {targetBone {boneName=spine_01}}]
    constraintWeights [1,1,1,1]
   }
   n29 = "Constraint - pouch_bk_01_r (PPPP)"!(
    ComponentPose=@n28
    Alpha=1
    ConstraintWeights_0=0.9867
    ConstraintWeights_1:float=0.0037
    ConstraintWeights_2:float=0.0048
    ConstraintWeights_3:float=0.0048
   ) -> (Pose) {
    alphaBoolBlend {blendOption=Linear}
    boneToModify {boneName=pouch_bk_01_r}
    constraintSetup [{targetBone {boneName=pelvis}}, {targetBone {boneName=spine_02}}, {targetBone {boneName=spine_03}}, {targetBone {boneName=spine_01}}]
    constraintWeights [1,1,1,1]
   }
   n30 = "Constraint - pouch_fr_01_l (P)"!(ComponentPose=@n29, Alpha=1, ConstraintWeights_0=1, ConstraintWeights_1:float=0) -> (Pose) {
    alphaBoolBlend {blendOption=Linear}
    boneToModify {boneName=pouch_fr_01_l}
    constraintSetup [{targetBone {boneName=pelvis}}]
    constraintWeights [1,0]
   }
   n31 = "Constraint - pouch_01_l (PPPP)"!(ComponentPose=@n30, Alpha=1, ConstraintWeights_0=1, ConstraintWeights_1:float=0, ConstraintWeights_2:float=0, ConstraintWeights_3:float=0) -> (Pose) {
    alphaBoolBlend {blendOption=Linear}
    boneToModify {boneName=pouch_01_l}
    constraintSetup [{targetBone {boneName=pelvis}}, {targetBone {boneName=spine_02}}, {targetBone {boneName=spine_03}}, {targetBone {boneName=spine_01}}]
    constraintWeights [1,1,1,1]
   }
   n32 = "Constraint - pouch_bk_01_l (PPPP)"!(
    ComponentPose=@n31
    Alpha=1
    ConstraintWeights_0=0.9867
    ConstraintWeights_1:float=0.0037
    ConstraintWeights_2:float=0.0048
    ConstraintWeights_3:float=0.0048
   ) -> (Pose) {
    alphaBoolBlend {blendOption=Linear}
    boneToModify {boneName=pouch_bk_01_l}
    constraintSetup [{targetBone {boneName=pelvis}}, {targetBone {boneName=spine_02}}, {targetBone {boneName=spine_03}}, {targetBone {boneName=spine_01}}]
    constraintWeights [1,1,1,1]
   }
   n33 = "Constraint - dyn_strap_02_r_MSTR (PPPPPPPPPP)"!(
    ComponentPose=@n32
    Alpha=1
    ConstraintWeights_0=0
    ConstraintWeights_1:float=0.1163
    ConstraintWeights_2:float=0
    ConstraintWeights_3:float=0.0886
    ConstraintWeights_4:float=0.2344
    ConstraintWeights_5:float=0.2506
    ConstraintWeights_6:float=0.1363
    ConstraintWeights_7:float=0.0373
    ConstraintWeights_8:float=0.0014
    ConstraintWeights_9:float=0.1351
   ) -> (Pose) {
    alphaBoolBlend {blendOption=Linear}
    boneToModify {boneName=dyn_strap_02_r_MSTR}
    constraintSetup [
     {targetBone {boneName=upperarm_twist_01_r}}
     {targetBone {boneName=pectoralis_02_r}}
     {targetBone {boneName=upperarm_twist_02_r}}
     {targetBone {boneName=spine_01}}
     {targetBone {boneName=spine_02}}
     {targetBone {boneName=spine_03}}
     {targetBone {boneName=spine_04}}
     {targetBone {boneName=spine_05}}
     {targetBone {boneName=clavicle_r}}
     {targetBone {boneName=pelvis}}
    ]
    constraintWeights [1,1,1,1,1,1,1,1,1,1]
   }
   n34 = Anim Dynamics - Bone: dyn_strap_02_r(ComponentPose=@n33, Alpha=0.25, ExternalForce="0,0,0") {
    angularDampingOverride=1
    angularSpringConstant=1800
    bAngularSpring=true
    bGravityOverrideInSimSpace=true
    bOverrideAngularDamping=true
    gravityScale=0.35
    linearDampingOverride=0.99
    linearSpringConstant=2500
    simulationSpace=RootRelative
    alphaBoolBlend {blendOption=Linear}
    boundBone {boneName=dyn_strap_02_r}
    physicsBodyDefinitions [{
      sphereCollisionRadius=0
      boundBone {boneName=dyn_strap_02_r}
      boxExtents=<10 20 20>
      constraintSetup {angularLimitsMax=<2 2 2>, angularLimitsMin=<-2 -2 -2>, angularTarget=<0 0 0>, linearAxesMax=<1 1 1>}
      localJointOffset=<10 0 0>
     }]
   }
   n35 = Get RlegAlpha() -> (RlegAlpha:double)
   n36 = "Transform (Modify) Bone - Bone: pouch_fr_01_r"(ComponentPose=@n34, Alpha=@n35, Translation="-0.666,-6.225,-0.344", Rotation="0.000,-20.000,-10.001", Scale="1,1,1") {
    rotationMode=BMM_Additive
    rotationSpace=BCS_BoneSpace
    translationMode=BMM_Additive
    translationSpace=BCS_BoneSpace
    alphaBoolBlend {blendOption=Linear}
    boneToModify {boneName=pouch_fr_01_r}
    rotation {roll=-10.001, yaw=-20}
    translation=<-0.666 -6.225 -0.344>
   }
   n37 = "Transform (Modify) Bone - Bone: pouch_01_r"(ComponentPose=@n36, Alpha=@n35, Translation="-0.466,-3.888,-1.094", Rotation="0.435,-15.019,-4.982", Scale="1,1,1") {
    rotationMode=BMM_Additive
    rotationSpace=BCS_BoneSpace
    translationMode=BMM_Additive
    translationSpace=BCS_BoneSpace
    alphaBoolBlend {blendOption=Linear}
    boneToModify {boneName=pouch_01_r}
    rotation=<rot 0.435 -15.0193 -4.982>
    translation=<-0.466 -3.888 -1.094>
   }
   n38 = "Transform (Modify) Bone - Bone: pouch_bk_01_r"(ComponentPose=@n37, Alpha=@n35, Translation="-0.101,-2.483,-0.879", Rotation="0.000,-5.000,-0.001", Scale="1,1,1") {
    rotationMode=BMM_Additive
    rotationSpace=BCS_BoneSpace
    translationMode=BMM_Additive
    translationSpace=BCS_BoneSpace
    alphaBoolBlend {blendOption=Linear}
    boneToModify {boneName=pouch_bk_01_r}
    rotation {roll=-0.001, yaw=-5.0001}
    translation=<-0.101 -2.483 -0.879>
   }
   n39 = Get bIsCrouching()
   n40 = "Transform (Modify) Bone - Bone: pouch_fr_01_l"!(ComponentPose=@n38, bAlphaBoolEnabled:bool=@n39, Translation="-0.739,4.810,-0.061", Rotation="0.000,20,10", Scale="1,1,1") -> (Pose) {
    alphaInputType=Bool
    rotationMode=BMM_Additive
    rotationSpace=BCS_BoneSpace
    translationMode=BMM_Additive
    translationSpace=BCS_BoneSpace
    alphaBoolBlend {blendOption=Linear}
    boneToModify {boneName=pouch_fr_01_l}
    rotation {roll=10, yaw=20}
    translation=<-0.739 4.81 -0.061>
   }
   n41 = "Transform (Modify) Bone - Bone: pouch_01_l"!(ComponentPose=@n40, bAlphaBoolEnabled:bool=@n39, Translation="-0.558,3.613,-1.103", Rotation="-4.546,14.584,4.998", Scale="1,1,1") -> (Pose) {
    alphaInputType=Bool
    rotationMode=BMM_Additive
    rotationSpace=BCS_BoneSpace
    translationMode=BMM_Additive
    translationSpace=BCS_BoneSpace
    alphaBoolBlend {blendOption=Linear}
    boneToModify {boneName=pouch_01_l}
    rotation=<rot -4.546 14.5843 4.998>
    translation=<-0.558 3.613 -1.103>
   }
   n42 = "Transform (Modify) Bone - Bone: pouch_bk_01_l"!(ComponentPose=@n41, bAlphaBoolEnabled:bool=@n39, Translation="0,2.483,-0.879", Rotation="0.000,5,0", Scale="1,1,1") -> (Pose) {
    alphaInputType=Bool
    rotationMode=BMM_Additive
    rotationSpace=BCS_BoneSpace
    translationMode=BMM_Additive
    translationSpace=BCS_BoneSpace
    alphaBoolBlend {blendOption=Linear}
    boneToModify {boneName=pouch_bk_01_l}
    rotation {yaw=5}
    translation {y=2.483, z=-0.879}
   }
   n43 = Look At - Bone: pectoralis_01_l!(ComponentPose=@n42, Alpha=1, LookAtLocation:Vector=Get LaimPecPos() -> (LaimPecPos:Vector)) -> (Pose) {
    alphaBoolBlend {blendOption=Linear}
    boneToModify {boneName=pectoralis_01_l}
    lookAtLocation {x=0}
    lookAtTarget {boneReference {boneName=upperarm_l}}
    lookAt_Axis {axis=<1 0>}
    lookUp_Axis {axis {x=0, z=1}}
   }
   n44 = Look At - Bone: pectoralis_01_r!(ComponentPose=@n43, Alpha=1, LookAtLocation:Vector=Get RaimPecPos() -> (RaimPecPos:Vector)) -> (Pose) {
    alphaBoolBlend {blendOption=Linear}
    boneToModify {boneName=pectoralis_01_r}
    lookAtLocation {x=0}
    lookAtTarget {boneReference {boneName=upperarm_r}}
    lookAt_Axis {axis=<1 0>}
    lookUp_Axis {axis {x=0, z=1}}
   }
   n45 = Get LarmDistanceBICEPAlpha() -> (LarmDistanceBICEPAlpha:double)
   n46 = "Transform (Modify) Bone - Bone: bicep_02_l"(ComponentPose=@n44, Alpha=@n45, Translation="-6,-0.000,-0.000", Rotation="0,0,0", Scale="1,1,1") {
    rotationMode=BMM_Additive
    rotationSpace=BCS_BoneSpace
    scaleMode=BMM_Additive
    scaleSpace=BCS_BoneSpace
    translationMode=BMM_Additive
    translationSpace=BCS_BoneSpace
    alphaBoolBlend {blendOption=Linear}
    boneToModify {boneName=bicep_02_l}
    translation {x=-6}
   }
   n47 = "Transform (Modify) Bone - Bone: bicep_01_l"(ComponentPose=@n46, Alpha=@n45, Translation="-0.000,0.000,2.343", Rotation="0,0,0", Scale="1.000,1.500,1.250") {
    rotationMode=BMM_Additive
    rotationSpace=BCS_BoneSpace
    scaleMode=BMM_Additive
    scaleSpace=BCS_BoneSpace
    translationMode=BMM_Additive
    translationSpace=BCS_BoneSpace
    alphaBoolBlend {blendOption=Linear}
    boneToModify {boneName=bicep_01_l}
    scale {y=1.5, z=1.25}
    translation {z=2.343}
   }
   n48 = "Transform (Modify) Bone - Bone: tricep_02_l"(ComponentPose=@n47, Alpha=@n45, Translation="-6,-0.000,-0.000", Rotation="0,0,0", Scale="1,1,1") {
    rotationMode=BMM_Additive
    rotationSpace=BCS_BoneSpace
    scaleMode=BMM_Additive
    scaleSpace=BCS_BoneSpace
    translationMode=BMM_Additive
    translationSpace=BCS_BoneSpace
    alphaBoolBlend {blendOption=Linear}
    boneToModify {boneName=tricep_02_l}
    translation {x=-6}
   }
   n49 = "Transform (Modify) Bone - Bone: tricep_01_l"(ComponentPose=@n48, Alpha=@n45, Translation="-0.000,0.000,2.343", Rotation="0,0,0", Scale="1.000,1.500,1.250") {
    rotationMode=BMM_Additive
    rotationSpace=BCS_BoneSpace
    scaleMode=BMM_Additive
    scaleSpace=BCS_BoneSpace
    translationMode=BMM_Additive
    translationSpace=BCS_BoneSpace
    alphaBoolBlend {blendOption=Linear}
    boneToModify {boneName=tricep_01_l}
    scale {y=1.5, z=1.25}
    translation {z=2.343}
   }
   n50 = Get RarmDistanceBICEPAlpha() -> (RarmDistanceBICEPAlpha:double)
   n51 = "Transform (Modify) Bone - Bone: bicep_02_r"(ComponentPose=@n49, Alpha=@n50, Translation="-6,-0.000,-0.000", Rotation="0,0,0", Scale="1,1,1") {
    rotationMode=BMM_Additive
    rotationSpace=BCS_BoneSpace
    scaleMode=BMM_Additive
    scaleSpace=BCS_BoneSpace
    translationMode=BMM_Additive
    translationSpace=BCS_BoneSpace
    alphaBoolBlend {blendOption=Linear}
    boneToModify {boneName=bicep_02_r}
    translation {x=-6}
   }
   n52 = "Transform (Modify) Bone - Bone: bicep_01_r"(ComponentPose=@n51, Alpha=@n50, Translation="-0.000,0.000,2.343", Rotation="0,0,0", Scale="1.000,1.500,1.250") {
    rotationMode=BMM_Additive
    rotationSpace=BCS_BoneSpace
    scaleMode=BMM_Additive
    scaleSpace=BCS_BoneSpace
    translationMode=BMM_Additive
    translationSpace=BCS_BoneSpace
    alphaBoolBlend {blendOption=Linear}
    boneToModify {boneName=bicep_01_r}
    scale {y=1.5, z=1.25}
    translation {z=2.343}
   }
   n53 = "Transform (Modify) Bone - Bone: tricep_02_r"(ComponentPose=@n52, Alpha=@n50, Translation="-6,-0.000,-0.000", Rotation="0,0,0", Scale="1,1,1") {
    rotationMode=BMM_Additive
    rotationSpace=BCS_BoneSpace
    scaleMode=BMM_Additive
    scaleSpace=BCS_BoneSpace
    translationMode=BMM_Additive
    translationSpace=BCS_BoneSpace
    alphaBoolBlend {blendOption=Linear}
    alphaScaleBiasClamp {scale=0.5}
    boneToModify {boneName=tricep_02_r}
    translation {x=-6}
   }
   n54 = "Transform (Modify) Bone - Bone: tricep_01_r"(ComponentPose=@n53, Alpha=@n50, Translation="-0.000,0.000,2.343", Rotation="0,0,0", Scale="1.000,1.500,1.250") {
    rotationMode=BMM_Additive
    rotationSpace=BCS_BoneSpace
    scaleMode=BMM_Additive
    scaleSpace=BCS_BoneSpace
    translationMode=BMM_Additive
    translationSpace=BCS_BoneSpace
    alphaBoolBlend {blendOption=Linear}
    alphaScaleBiasClamp {scale=0.5}
    boneToModify {boneName=tricep_01_r}
    scale {y=1.5, z=1.25}
    translation {z=2.343}
   }
   n55 = "Transform (Modify) Bone - Bone: upperarm_twist_01_r"(ComponentPose=@n54, Alpha=Get RshoulderAlpha() -> (RshoulderAlpha:double), Translation="0,0,0", Rotation="0,0,0", Scale="1.2,1.2,1.2") {
    rotationMode=BMM_Additive
    rotationSpace=BCS_BoneSpace
    scaleMode=BMM_Additive
    scaleSpace=BCS_BoneSpace
    translationMode=BMM_Additive
    translationSpace=BCS_BoneSpace
    alphaBoolBlend {blendOption=Linear}
    boneToModify {boneName=upperarm_twist_01_r}
    scale=<1.2 1.2 1.2>
   }
   n56 = "Transform (Modify) Bone - Bone: upperarm_twist_01_l"(ComponentPose=@n55, Alpha=Get LshoulderAlpha() -> (LshoulderAlpha:double), Translation="0,0,0", Rotation="0,0,0", Scale="1.2,1.2,1.2") {
    rotationMode=BMM_Additive
    rotationSpace=BCS_BoneSpace
    scaleMode=BMM_Additive
    scaleSpace=BCS_BoneSpace
    translationMode=BMM_Additive
    translationSpace=BCS_BoneSpace
    alphaBoolBlend {blendOption=Linear}
    boneToModify {boneName=upperarm_twist_01_l}
    scale=<1.2 1.2 1.2>
   }
   n57 = "Constraint - kneepad_01_r (PP)"!(ComponentPose=@n56, Alpha=1, ConstraintWeights_0=0.25, ConstraintWeights_1:float=0.75) -> (Pose) {
    alphaBoolBlend {blendOption=Linear}
    boneToModify {boneName=kneepad_01_r}
    constraintSetup [{targetBone {boneName=calf_r}}, {targetBone {boneName=thigh_r}}]
    constraintWeights [1,1]
   }
   n58 = "Constraint - kneepad_01_l (PP)"!(ComponentPose=@n57, Alpha=1, ConstraintWeights_0=0.25, ConstraintWeights_1:float=0.75) -> (Pose) {
    alphaBoolBlend {blendOption=Linear}
    boneToModify {boneName=kneepad_01_l}
    constraintSetup [{targetBone {boneName=calf_l}}, {targetBone {boneName=thigh_l}}]
    constraintWeights [1,1]
   }
   n59 = "Constraint - shoepad_01_l (P)"(ComponentPose=@n58, Alpha=0, ConstraintWeights_0=1) {
    alphaBoolBlend {blendOption=Linear}
    boneToModify {boneName=shoepad_01_l}
    constraintSetup [{targetBone {boneName=calf_twist_02_l}}]
    constraintWeights [1]
   }
   n60 = Flap Node - Bone: shoepad_01_l!(ComponentPose:Component Space Pose Link Structure=@n59, Alpha:float=1) -> (Pose:Component Space Pose Link Structure) {
    is=title FortAnimGraphNode_Flap
    bUseClamp=true
    targetClampMax=1
    targetComponent=RotationZ
    alphaBoolBlend {blendOption=Linear}
    sourceDrives [{
      bUseRange=false
      drivingCurve=null
      multiplier=0.8
      rangeMax=1
      rangeMin=-1
      remappedMax=1
      remappedMin=0
      sourceComponent=RotationZ
      useQuaternion=true
      sourceBone {boneName=foot_l}
     }]
    targetBone {boneName=shoepad_01_l}
   }
   n61 = "Constraint - shoepad_01_r (P)"(ComponentPose=@n60, Alpha=0, ConstraintWeights_0=1) {
    alphaBoolBlend {blendOption=Linear}
    boneToModify {boneName=shoepad_01_r}
    constraintSetup [{targetBone {boneName=calf_twist_02_r}}]
    constraintWeights [1]
   }
   n62 = Flap Node - Bone: shoepad_01_r!(ComponentPose:Component Space Pose Link Structure=@n61, Alpha:float=1) -> (Pose:Component Space Pose Link Structure) {
    is=title FortAnimGraphNode_Flap
    bUseClamp=true
    targetClampMax=1
    targetComponent=RotationZ
    alphaBoolBlend {blendOption=Linear}
    sourceDrives [{
      bUseRange=false
      drivingCurve=null
      multiplier=0.8
      rangeMax=1
      rangeMin=-1
      remappedMax=1
      remappedMin=0
      sourceComponent=RotationZ
      useQuaternion=true
      sourceBone {boneName=foot_r}
     }]
    targetBone {boneName=shoepad_01_r}
   }
   n63 = Get LshoulderRotateBackAlpha()
   n64 = "Transform (Modify) Bone - Bone: dyn_parent_shldrpad_l"(ComponentPose=@n62, Alpha=@n63, Translation="0,0,0", Rotation="0.000,0.000,-35", Scale="1,1,1") {
    rotationMode=BMM_Additive
    rotationSpace=BCS_BoneSpace
    alphaBoolBlend {blendOption=Linear}
    boneToModify {boneName=dyn_parent_shldrpad_l}
    rotation {roll=-35}
   }
   n65 = Look At - Bone: dyn_shldrpad_01_l(ComponentPose=@n64, Alpha=@n63) {
    alphaBoolBlend {blendOption=Linear}
    boneToModify {boneName=dyn_shldrpad_01_l}
    lookAtLocation {x=0}
    lookAtTarget {boneReference {boneName=dyn_shldrpad_01_l_aim}}
    lookAt_Axis {axis=<1 0>}
    lookUp_Axis {axis {x=0, z=1}}
   }
   n66 = "Transform (Modify) Bone - Bone: dyn_shldrpad_01_l"(ComponentPose=@n65, Alpha=Get isTargetingAlpha() -> (isTargetingAlpha:double), Translation="0,0,0", Rotation="3,0.000,0.000", Scale="1,1,1") {
    rotationMode=BMM_Additive
    rotationSpace=BCS_BoneSpace
    translationMode=BMM_Additive
    translationSpace=BCS_BoneSpace
    alphaBoolBlend {blendOption=Linear}
    alphaScaleBiasClamp {bInterpResult=true}
    boneToModify {boneName=dyn_shldrpad_01_l}
    rotation {pitch=3}
   }
   n67 = "Constraint - dyn_shldrpad_03_r (PPPPPPPPPPP)"!(
    ComponentPose=@n66
    Alpha=Get rifleAimAlpha() -> (rifleAimAlpha:double)
    ConstraintWeights_0=0.0078
    ConstraintWeights_1:float=0.0309
    ConstraintWeights_2:float=0
    ConstraintWeights_3:float=0.0032
    ConstraintWeights_4:float=0.0003
    ConstraintWeights_5:float=0.0267
    ConstraintWeights_6:float=0.1842
    ConstraintWeights_7:float=0.074
    ConstraintWeights_8:float=0.0835
    ConstraintWeights_9:float=0.5888
    ConstraintWeights_10:float=0.0006
   ) -> (Pose) {
    alphaBoolBlend {blendOption=Linear}
    boneToModify {boneName=dyn_shldrpad_03_r}
    constraintSetup [
     {targetBone {boneName=upperarm_r}}
     {targetBone {boneName=upperarm_twist_01_r}}
     {targetBone {boneName=pectoralis_02_r}}
     {targetBone {boneName=upperarm_twist_02_r}}
     {targetBone {boneName=spine_01}}
     {targetBone {boneName=spine_02}}
     {targetBone {boneName=spine_03}}
     {targetBone {boneName=spine_04}}
     {targetBone {boneName=spine_05}}
     {targetBone {boneName=clavicle_r}}
     {targetBone {boneName=pelvis}}
    ]
    constraintWeights [1,1,1,1,1,1,1,1,1,1,1]
   }
   n68 = "Transform (Modify) Bone - Bone: dyn_parent_shldrpad_r"(ComponentPose=@n67, Alpha=Get RshoulderRotateBackAlpha(), Translation="0,0,0", Rotation="0.000,0.000,35", Scale="1,1,1") {
    rotationMode=BMM_Additive
    rotationSpace=BCS_BoneSpace
    alphaBoolBlend {blendOption=Linear}
    boneToModify {boneName=dyn_parent_shldrpad_r}
    rotation {roll=35}
   }
   n69 = Get isTargetingAlpha()
   n70 = "Transform (Modify) Bone - Bone: dyn_shldrpad_01_r"!(ComponentPose=@n68, Alpha:float=@n69, Translation="0,0,0", Rotation="10,0.000,0.000", Scale="1,1,1") -> (Pose) {
    rotationMode=BMM_Additive
    rotationSpace=BCS_BoneSpace
    translationMode=BMM_Additive
    translationSpace=BCS_BoneSpace
    alphaBoolBlend {blendOption=Linear}
    alphaScaleBiasClamp {bInterpResult=true}
    boneToModify {boneName=dyn_shldrpad_01_r}
    rotation {pitch=10}
   }
   n71 = "Transform (Modify) Bone - Bone: dyn_shldrpad_03_r"!(ComponentPose=@n70, Alpha:float=@n69, Translation="-0.000,-5.216,0.000", Rotation="9.392,-3.451,-20.283", Scale="1,1,1") -> (Pose) {
    rotationMode=BMM_Additive
    rotationSpace=BCS_BoneSpace
    scaleMode=BMM_Replace
    scaleSpace=BCS_BoneSpace
    translationMode=BMM_Additive
    translationSpace=BCS_BoneSpace
    alphaBoolBlend {blendOption=Linear}
    alphaScaleBiasClamp {bInterpResult=true}
    boneToModify {boneName=dyn_shldrpad_03_r}
    rotation=<rot 9.392 -3.451 -20.2829>
    translation {y=-5.216}
   }
   n72 = "Transform (Modify) Bone - Bone: dyn_shldrpad_03_r"(ComponentPose=@n71, Alpha=@n69, Translation="0,0,0", Rotation="0,0,0", Scale="0.9,1,1") {
    rotationMode=BMM_Additive
    rotationSpace=BCS_BoneSpace
    scaleMode=BMM_Replace
    scaleSpace=BCS_BoneSpace
    translationMode=BMM_Additive
    translationSpace=BCS_BoneSpace
    alphaBoolBlend {blendOption=Linear}
    alphaScaleBiasClamp {bInterpResult=true}
    boneToModify {boneName=dyn_shldrpad_03_r}
    scale {x=0.9}
   }
   Output Pose(Result=Component To Local(ComponentPose=@n72)) {layerGroup=None}
   Get rifleAimAlpha()
  ]}
 EventGraph {nodes [
   Event BlueprintUpdateAnimation()
   > n1 = Sequence!(execute) -> (then_0, then_1, then_2:Exec, then_3:Exec, then_4:Exec, then_5:Exec, then_6:Exec)
   n2 = GetPartSkeletalMeshComponent()
   > Set RshoulderRotateBackAlpha(RshoulderRotateBackAlpha:double=GetAlphaFromReferenceTransformDelta(
     SkeletalMeshComponent=@n2
     BoneName=upperarm_r
     BaseName=clavicle_r
     DeltaComponentType=QuaternionTwist
     InputRange=(X=-45,Y=-125)
     TwistAxis="1,0,0"
    )) -> (Output_Get:double)
   > Set rifleAimAlpha(rifleAimAlpha:double=SelectFloat(
     A=GetAlphaFromReferenceTransformDelta(
      SkeletalMeshComponent=@n2
      BoneName=upperarm_r
      BaseName=clavicle_r
      DeltaComponentType=QuaternionTwist
      InputRange=(X=0,Y=-25)
      OutputRange=(X=0.5,Y=0)
      TwistAxis="1,0,0"
     )
     B=0
     bPickA=Get isAimingRifle() -> (isAimingRifle:bool)
    )) -> (Output_Get:double)
   n3 = GetAlphaFromReferenceTransformDelta(
    SkeletalMeshComponent=@n2
    BoneName=upperarm_l
    BaseName=clavicle_l
    DeltaComponentType=QuaternionTwist
    InputRange=(X=-45,Y=-125)
    TwistAxis="1,0,0"
   )
   Set LshoulderRotateBackAlpha(execute=@n1.then_1, LshoulderRotateBackAlpha:double=@n3) -> (Output_Get:double)
   > Set LshoulderRotateBackAlpha_Backbling(LshoulderRotateBackAlpha_Backbling:double=SelectFloat(A=@n3, B=0, bPickA=Get isWearingBackbling())) -> (Output_Get:double)
   n4 = GetPartSkeletalMeshComponent()
   n5 = GetAlphaFromReferenceTransformDelta(
    SkeletalMeshComponent=@n4
    BoneName=lowerarm_r
    BaseName=upperarm_r
    DeltaComponentType=QuaternionTwist
    InputRange=(X=0,Y=-90)
    OutputRange=(X=0,Y=0.7)
    TwistAxis="0,0,1"
   )
   Set RflexAlpha(
    execute=@n1.then_2
    RflexAlpha:double=SelectFloat(A=0.25, B=MapRangeClamped(Value=@n5, InRangeA=0, InRangeB=1, OutRangeA=0.2, OutRangeB=0.7), bPickA=Get bIsTargeting())
   ) -> (Output_Get:double)
   n6 = GetAlphaFromReferenceTransformDelta(
    SkeletalMeshComponent=@n4
    BoneName=upperarm_r
    BaseName=clavicle_r
    DeltaComponentType=QuaternionTwist
    InputRange=(X=0,Y=-128)
    TwistAxis="0,0,1"
   )
   > Set RshoulderAlpha(RshoulderAlpha:double=@n6) -> (Output_Get:double)
   > Set RaimPecPos(RaimPecPos:Vector=MakeVector(Y=0, Z=MapRangeClamped(Value=@n6, InRangeA=0, InRangeB=1, OutRangeB=-10))) -> (Output_Get:Vector)
   n7 = GetAlphaFromReferenceTransformDelta(
    SkeletalMeshComponent=@n4
    BoneName=upperarm_l
    BaseName=clavicle_l
    DeltaComponentType=QuaternionTwist
    InputRange=(X=-20,Y=-75)
    TwistAxis="0,0,1"
   )
   > Set LshoulderAlpha(LshoulderAlpha:double=@n7) -> (Output_Get:double)
   > Set LaimPecPos(LaimPecPos:Vector=MakeVector(Y=0, Z=MapRangeClamped(Value=@n7, InRangeA=0, InRangeB=1, OutRangeB=10))) -> (Output_Get:Vector)
   > Set RlegAlpha(RlegAlpha:double=SelectFloat(
     A=1
     B=GetAlphaFromReferenceTransformDelta(SkeletalMeshComponent=@n4, BoneName=thigh_r, BaseName=pelvis, DeltaComponentType=QuaternionTwist, InputRange=(X=0,Y=100), TwistAxis="0,0,1")
     bPickA=Get bIsCrouching()
    )) -> (Output_Get:double)
   > Set LbicepFlex(LbicepFlex:double=GetAlphaFromReferenceTransformDelta(
     SkeletalMeshComponent=@n4
     BoneName=lowerarm_l
     BaseName=upperarm_l
     DeltaComponentType=QuaternionTwist
     InputRange=(X=-20,Y=-70)
     TwistAxis="0,0,1"
    )) -> (Output_Get:double)
   > Set RbicepFlex(RbicepFlex:double=GetAlphaFromReferenceTransformDelta(
     SkeletalMeshComponent=@n4
     BoneName=lowerarm_r
     BaseName=upperarm_r
     DeltaComponentType=QuaternionTwist
     InputRange=(X=-20,Y=-70)
     TwistAxis="0,0,1"
    )) -> (Output_Get:double)
   n8 = GetPartSkeletalMeshComponent()
   n9 = TransformFromBoneSpace!(
    execute=@n1.then_3
    self:Skinned Mesh Component Object Reference=@n8
    BoneName:Name=upperarm_l
    InPosition:Vector="0, 0, 0"
    InRotation:Rotator="0, 0, 0"
   ) -> (then, OutPosition:Vector, OutRotation:Rotator)
   > n10 = TransformFromBoneSpace(self=@n8, BoneName=lowerarm_twist_02_l, InPosition="0, 0, 0", InRotation="0, 0, 0")
   > Set LarmDistanceBICEPAlpha(LarmDistanceBICEPAlpha:double=MapRangeClamped(Value="Distance (Vector)"!(V1:Vector=@n9, V2:Vector=@n10) -> (ReturnValue:double), InRangeA=27, InRangeB=34, OutRangeA=1, OutRangeB=0)) -> (Output_Get:double)
   > n11 = TransformFromBoneSpace(self=@n8, BoneName=upperarm_r, InPosition="0, 0, 0", InRotation="0, 0, 0")
   > n12 = TransformFromBoneSpace(self=@n8, BoneName=lowerarm_twist_02_r, InPosition="0, 0, 0", InRotation="0, 0, 0")
   > Set RarmDistanceBICEPAlpha(RarmDistanceBICEPAlpha:double=MapRangeClamped(Value="Distance (Vector)"(V1=@n11, V2=@n12), InRangeA=27, InRangeB=34, OutRangeA=1, OutRangeB=0)) -> (Output_Get:double)
   n13 = TransformFromBoneSpace(execute=@n1.then_4, self=@n8, BoneName=upperarm_r, InPosition="0, 0, 0", InRotation="0, 0, 0")
   > n14 = TransformFromBoneSpace(self=@n8, BoneName=hand_r, InPosition="0, 0, 0", InRotation="0, 0, 0")
   > Set RarmDistanceAlpha(RarmDistanceAlpha:double=MapRangeClamped(Value="Distance (Vector)"(V1=@n13, V2=@n14), InRangeA=40, InRangeB=54, OutRangeA=0.7, OutRangeB=0)) -> (Output_Get:double)
   Set isTargetingAlpha(execute=@n1.then_5, isTargetingAlpha:double="To Float (Boolean)"(InBool=Get bIsTargeting())) -> (Output_Get:double)
   n15 = Cast To FortPlayerAnimInstance(execute=@n1.then_6, Object=GetAnimInstance(self=GetOwnerSkeletalMeshComponent()))
   > Set isAimingPistol(isAimingPistol:double=0) -> (Output_Get:double)
   > Set isAimingRifle(isAimingRifle:bool=false) -> (Output_Get:bool)
   > n16 = Switch on EFortWeaponCoreAnimation!(execute, Selection:EFortWeaponCoreAnimation Enum=Get WeaponCoreAnim(self=@n15)) -> (
    Melee:Exec
    Pistol:Exec
    Shotgun:Exec
    PaperBlueprint:Exec
    Rifle:Exec
    MeleeOneHand:Exec
    MachinePistol:Exec
    RocketLauncher:Exec
    GrenadeLauncher:Exec
    GoingCommando:Exec
    AssaultRifle:Exec
    TacticalShotgun:Exec
    SniperRifle:Exec
    TrapPlacement:Exec
    ShoulderLauncher:Exec
    AbilityDecoTool:Exec
    Crossbow:Exec
    C4:Exec
    RemoteControl:Exec
    DualWield:Exec
    AR_BullPup:Exec
    AR_ForwardGrip:Exec
    MedPackPaddles:Exec
    SMG_P90:Exec
    AR_DrumGun:Exec
    Consumable_Small:Exec
    Consumable_Large:Exec
    Balloon:Exec
    MountedTurret:Exec
    CreativeTool:Exec
    ExplosiveBow:Exec
    AshtonIndigo:Exec
    AshtonChicago:Exec
    MeleeDualWield:Exec
    Unarmed:Exec
   )
   Branch(execute=[
     @n16.AssaultRifle
     @n16.Crossbow
     @n16.GrenadeLauncher
     @n16.Rifle
     @n16.RocketLauncher
     @n16.Shotgun
     @n16.SniperRifle
     @n16.TacticalShotgun
    ])
   > Set isAimingRifle(isAimingRifle=true)
   Branch(execute=[@n16.MachinePistol, @n16.Pistol])
   > Set isAimingPistol(isAimingPistol=1)
   Event OnCharacterCustomizationLoaded()
   > Do Once()
   n17 = CheckForBackbling()
   > Set isWearingBackbling(isWearingBackbling=OR Boolean(A=@n17, B=@n17.bIsBackblingLong))
   DrawDebugLine!(
    execute
    LineStart:Vector=@n13
    LineEnd:Vector=@n14
    LineColor:Linear Color Structure=(R=0.026,G=1,B=0,A=1)
    Duration:float=0
    Thickness:float=1
    DepthPriority:EDrawDebugSceneDepthPriorityGroup Enum=World
   ) -> (then)
   Get isAimingRifle()
  ]}
]
variables [
 isAimingPistol {type=number}
 isAimingRifle {type=boolean}
 isTargetingAlpha {type=number}
 isWearingBackbling {type=boolean}
 LaimPecPos {type=Vector}
 LarmDistanceBICEPAlpha {type=number}
 LbicepFlex {type=number}
 LshoulderAlpha {type=number}
 LshoulderRotateBackAlpha {type=number}
 LshoulderRotateBackAlpha_Backbling {type=number}
 RaimPecPos {type=Vector}
 RarmDistanceAlpha {type=number}
 RarmDistanceBICEPAlpha {type=number}
 RbicepFlex {type=number}
 RflexAlpha {type=number}
 rifleAimAlpha {type=number}
 RlegAlpha {type=number}
 RshoulderAlpha {type=number}
 RshoulderRotateBackAlpha {type=number}
]
