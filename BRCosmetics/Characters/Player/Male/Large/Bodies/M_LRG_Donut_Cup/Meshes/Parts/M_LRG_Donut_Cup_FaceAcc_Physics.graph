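# Physics asset for the face accessory part of the large male Donut Cup player body, defining bodies for front hair bones.
# A near-point capsule on the MSTR bone anchors the front root bone through a tight swing and twist constraint.
# Two right-side bones, bottom and middle, branch from that root with much looser swing and equally tight twist.
# The three hair bones share identical short, fat capsules.
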PhysicsAsset /BRCosmetics/Characters/Player/Male/Large/Bodies/M_LRG_Donut_Cup/Meshes/Parts/M_LRG_Donut_Cup_FaceAcc_Physics
bodies {
 dyn_hair_fr_00 [{shapeType=Capsule, length=1.5, radius=1.75, center=<0.6642 0 0>, rotation=<rot 90 0.0198 0.0198>}]
 dyn_hair_fr_00_MSTR [{shapeType=Capsule, length=0.01, radius=0.01}]
 dyn_hair_fr_bt_01_r [{shapeType=Capsule, length=1.5, radius=1.75, center=<0.6642 0 0>, rotation=<rot 90 0.0198 0.0198>}]
 dyn_hair_fr_mid_02_r [{shapeType=Capsule, length=1.5, radius=1.75, center=<0.6642 0 0>, rotation=<rot 90 0.0198 0.0198>}]
}
constraints [
 {bone1Name=dyn_hair_fr_bt_01_r, bone2Name=dyn_hair_fr_00, swing1LimitDegrees=45, swing2LimitDegrees=45, twistLimitDegrees=5}
 {bone1Name=dyn_hair_fr_mid_02_r, bone2Name=dyn_hair_fr_00, swing1LimitDegrees=45, swing2LimitDegrees=45, twistLimitDegrees=5}
 {bone1Name=dyn_hair_fr_00, bone2Name=dyn_hair_fr_00_MSTR, swing1LimitDegrees=10, swing2LimitDegrees=10, twistLimitDegrees=5}
]
